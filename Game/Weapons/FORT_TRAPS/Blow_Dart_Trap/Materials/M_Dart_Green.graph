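# M_Dart_Green is a simple Default Lit material in the Blow Dart Trap weapon assets, named as the green variant for the dart.
# It holds a Color vector parameter defaulting to dark green and two scalar constants, with no textures.
# The material is flagged for use with Nanite.
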Material /Game/Weapons/FORT_TRAPS/Blow_Dart_Trap/Materials/M_Dart_Green
graph {
 properties {bUsedWithNanite=true, pixelDepthOffsetMode=PDOM_Legacy, usedShadingModels=Default Lit}
 nodes [
  Constant() {r=0.4}
  Constant() {r=0.8}
  VectorParameter() {parameterName=Color, defaultValue=<rgba 0.0163 0.349 0.0377 0>}
 ]
}
